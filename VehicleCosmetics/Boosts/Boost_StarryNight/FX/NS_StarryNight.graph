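# Niagara system for the Starry Night vehicle boost cosmetic, attached to filtered skeletal-mesh sockets.
# A counter-rotating mesh cone, glow sprites and drifting, spinning star sprites accompany three layered ribbon trails.
# Color variant arrays tint the cone, while painted-variant overrides color the ribbons and other sprites.
# The ribbons use per-layer width curves, and a computed sine/cosine vertical wave scales its amplitude with camera alignment to the owner's forward axis.
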NiagaraSystem /VehicleCosmetics/Boosts/Boost_StarryNight/FX/NS_StarryNight
dependencies {
 usedDataInterfaces [
  &NiagaraDataInterfaceColorCurve
  &NiagaraDataInterfaceCamera
  &NiagaraDataInterfaceCurve
  &NiagaraDataInterfaceArrayColor
  &NiagaraDataInterfaceSkeletalMesh
 ]
 usedDynamicInputs []
 usedModules [
  &self:SystemSpawnScript.SystemScriptSource.SystemScriptGraph.NiagaraNodeAssignment_0.SetVariables_#1
  &SystemState
 ]
}
emitters {
 NE_VC_Boost_MeshCone001 {
  emitterSpawnScript [/SetVariables(
    Emitter.BaseTint:LinearColor=&SelectLinearColorFromArray(
     Array Sampling Mode=::NewEnumerator1
     Direct Array Index=$System.ColorVariantID
     Color Selection Array=DataInterface() {propertyValues {ColorData [
        <rgba 0.9948 0.8177 0.0415 1>
        <rgba 1 1 1 1>
        <rgba 0.4219 0.4219 0.4219 1>
        <rgba 0.6302 0.3455 0.2593 1>
        <rgba 0.4479 0.4564 0.8958 1>
        <rgba 1.1 0.4431 0.4114 1>
        <rgba 0.1128 0.8333 0.1488 0.9998>
        <rgba 0.5677 0.5677 0.5677 1>
        <rgba 0.5014 1 0 1>
        <rgba 1 0.4573 0.1667 1>
        <rgba 2 0.044 1.2986 1>
        <rgba 0.6599 0.4413 0.974 1>
        <rgba 1 0.7662 0 1>
        <rgba 0.0273 0.8375 1.2396 1>
        <rgba 0.7396 0.7396 0.7396 0>
       ]}}
     Random Seed=?
    )
    Emitter.EmissiveColor:LinearColor=&SelectLinearColorFromArray(
     Array Sampling Mode=::NewEnumerator1
     Direct Array Index=$System.ColorVariantID
     Color Selection Array=DataInterface() {propertyValues {ColorData [
        <rgba 0.9531 0.5418 0.0212 1>
        <rgba 1 1 1 1>
        <rgba 0.0885 0.0885 0.0885 1>
        <rgba 0.8542 0.4914 0.3915 1>
        <rgba 0.0447 0.0618 0.953 1>
        <rgba 1 0.103 0.0885 1>
        <rgba 0.0391 0.6823 0.1014 1>
        <rgba 0.0677 0.0677 0.0677 1>
        <rgba 0.5507 1 0.099 0.9565>
        <rgba 2 1.4022 0 1>
        <rgba 0.953 0.3326 0.7305 1>
        <rgba 0.8468 0.044 2 1>
        <rgba 0.9479 0.7263 0 1>
        <rgba 0 0.6683 1 1>
        <rgba 0.7135 0.7135 0.7135 0>
       ]}}
     Random Seed=?
    )
    Emitter.FresnelColor:LinearColor=&SelectLinearColorFromArray(
     Array Sampling Mode=::NewEnumerator1
     Direct Array Index=$System.ColorVariantID
     Color Selection Array=DataInterface() {propertyValues {ColorData [
        <rgba 0.7865 0.447 0.0175 1>
        <rgba 1 1 1 1>
        <rgba 0.3177 0.3177 0.3177 1>
        <rgba 0.5677 0.4247 0.3814 1>
        <rgba 0.1797 0.1905 0.75 1>
        <rgba 1 0.8914 0.8948 1>
        <rgba 0.3716 0.4583 0.3628 1>
        <rgba 0.786 0.786 0.786 1>
        <rgba 0.3854 0.7031 0.0659 1>
        <rgba 0.786 0.2854 0.0173 1>
        <rgba 0.7835 0.0176 0.5089 1>
        <rgba 0.786 0.0157 0.3836 1>
        <rgba 0.786 0.6062 0.0173 1>
        <rgba 0.0173 0.531 0.786 1>
        <rgba 0.4531 0.4531 0.4531 0>
       ]}}
     Random Seed=?
    )
   ) {isSetParameters=true}]
  emitterUpdateScript [
   &EmitterState(Inactive Response=::NewEnumerator1, Life Cycle Mode=::NewEnumerator1, Loop Duration=1, UseLoopDelay=false)
   &VC_FilteredSocketCount()
   &SpawnBurst_Instantaneous(Spawn Count=2, Use Spawn Probability=false, UseLoopCountLimit=false)
   &SpawnRate(SpawnRate=6, Use Spawn Probability=false) {enabled=false}
  ]
  particleSpawnScript [
   &VC_PaintedVariantsOverride(DefaultColor=<rgba 0.25 0 0.0836 1>) {enabled=false}
   &InitializeParticle(
    Color=$Emitter.BaseTint
    Lifetime=&MergeHelper_FloatSwitch(0=3, 1=3, 2=3, 3=?, Index=$User.GameModeID, Num=::NewEnumerator1)
    Mesh Scale Mode=::NewEnumerator3
    Mesh Uniform Scale=0.22
    Position Mode=::NewEnumerator0
    Sprite Rotation Mode=::NewEnumerator2
    Write Material Random=true
   )
   &PartitionParticles(Number Of Partitions=$Emitter.PortCount)
   &SkeletalMeshLocation(
    Bone / Socket Index=$Particles.PartitionParticles.Partition
    Bone / Socket Sampling Mode=::NewEnumerator5
    MaskSpawnGroup=false
    Mesh Sampling Type=::NewEnumerator6
    Sampled Position Offset=&AddVectorToPosition(
     Position=$User.DTCarOffset002
     Vector=&MergeHelper_VectorSwitch(0=<-46 0 0>, 1=<-46 0 0>, 2=<-46 0 0>, 3=?, Index=$User.GameModeID, Num=::NewEnumerator1)
    )
    Skeletal Mesh=$System.SkeletalMesh
   )
   &AddVelocity(Velocity=<0 0 0>) {enabled=false}
  ]
  particleUpdateScript [
   &InheritSourceMovement(ApplyForce=false, ApplyPosition=true, ApplyPositionFalloff=false) {enabled=false}
   &ParticleState()
   &MeshRotationForce(Rotation=&MaskVector(Mask Source=1, Mask Parameter=$Particles.PartitionParticles.Partition, Mask=0, Masked Vector=<2 0 0>, Passthrough Vector=<-2 0 0>))
   &SolveForcesAndVelocity(Clamp Velocity=false, Limit Acceleration=false)
   &DynamicMaterialParameters(
    Index 0 Param 1=1
    Index 0 Param 2=15
    Module.Index0Float0Write=true
    Module.Index0Float1Write=true
    Module.Index0Float2Write=true
    Module.Index0Float3Write=true
   )
  ]
  rendererClasses [&NiagaraMeshRendererProperties, &NiagaraSpriteRendererProperties]
  renderers [
   {rendererClass=&NiagaraMeshRendererProperties, rendererIndex=0}
   {rendererClass=&NiagaraSpriteRendererProperties, rendererIndex=1}
  ]
 }
 NE_VC_Boost_Ribbons {
  emitterUpdateScript [
   &EmitterState(Life Cycle Mode=::NewEnumerator1, Loop Duration=1, UseLoopDelay=false)
   &VC_FilteredSocketCount()
   &VC_RateQualityHandler(Epic Spawn Rate=100, High Spawn Rate=100, Low Spawn Rate=100, Medium Spawn Rate=100, Use Custom Rate?=true)
   &VC_SpawnRibbon_v1(Number of Ribbons=$Emitter.PortCount, Spawn Time=-1, SpawnRate=$Emitter.Spawn Rate)
  ]
  particleSpawnScript [
   &VC_PaintedVariantsOverride(Orange=<rgba 0.6724 0.0883 0.2836 1>, Use R1 Painted Overrides=true, Use R2 Painted Overrides=true) {enabled=false}
   &VC_InitializeRibbon_v1(
    Custom Offset=true
    Enable Skeletal Position=true
    Facing Axis:ENiagaraOrientationAxis=::YAxis
    Lifetime=&MergeHelper_FloatSwitch(0=1, 1=1, 2=1, 3=?, Index=$User.GameModeID, Num=::NewEnumerator1)
    Local Vector Offset=&AddVectorToPosition(Position=<-38.5525 0 0>, Vector=Linked() {linkedVariable {name=User.DTCarOffset002, type=NiagaraPosition}})
    Offset Amount=1
    Ribbon Width=90
   )
   &AddVelocity(Velocity=<-1000 0 0>) {module=AddVelocity001}
  ]
  particleUpdateScript [
   &VC_PaintedVariantsOverride(
    Orange=<rgba 0.6724 0.0883 0.2836 1>
    R1 Black:LinearColor=&ColorFromCurve(ColorCurve=DataInterface() {propertyValues {
       AlphaCurve {keys [<at 0 4 RCIM_Cubic>,<at 0.1045 0.9695>,<at 1 0 RCIM_Cubic>]}
       BlueCurve {keys [<at 0 0.035 RCIM_Cubic>,<at 0.6713 0.0208>]}
       GreenCurve {keys [<at 0 0.035 RCIM_Cubic>,<at 0.6713 0.0208>]}
       RedCurve {keys [<at 0 0.035 RCIM_Cubic>,<at 0.6713 0.0208>]}
      }})
    R1 BurntSienna:LinearColor=&ColorFromCurve(ColorCurve=DataInterface() {propertyValues {
       AlphaCurve {keys [<at 0 4 RCIM_Cubic>,<at 0.2166 2>,<at 1 0 RCIM_Cubic>]}
       BlueCurve {keys [<at 0 0.0284>,<at 0.6066 0.0061>]}
       GreenCurve {keys [<at 0 0.0717>,<at 0.6066 0.0193>]}
       RedCurve {keys [<at 0 0.2292>,<at 0.6066 0.067>]}
      }})
    R1 Cobalt:LinearColor=&ColorFromCurve(ColorCurve=DataInterface() {propertyValues {
       AlphaCurve {keys [<at 0 4 RCIM_Cubic>,<at 0.6217 1>,<at 1 0 RCIM_Cubic>]}
       BlueCurve {keys [<at 0 0.6867 RCIM_Cubic>,<at 0.6282 0.067>]}
       GreenCurve {keys [<at 0 0.0551 RCIM_Cubic>,<at 0.6282 0.0013>]}
       RedCurve {keys [<at 0 0.0429 RCIM_Cubic>,<at 0.6282 0>]}
      }})
    R1 Crimson:LinearColor=&ColorFromCurve(ColorCurve=DataInterface() {propertyValues {
       AlphaCurve {keys [<at 0 4 RCIM_Cubic>,<at 1 0 RCIM_Cubic>]}
       BlueCurve {keys [<at 0.0162 0.118>,<at 0.5441 0.0018>]}
       GreenCurve {keys [<at 0.0162 0.1069>,<at 0.5441 0>]}
       RedCurve {keys [<at 0.0162 0.526>,<at 0.5441 0.067>]}
      }})
    R1 DefaultColor=&ColorFromCurve(ColorCurve=DataInterface() {propertyValues {
       AlphaCurve {keys [<at 0 4 RCIM_Cubic>,<at 0.9208 0.9976>,<at 1 0 RCIM_Cubic>]}
       BlueCurve {keys [<at 0 0.552>,<at 0.6099 0.0666>]}
       GreenCurve {keys [<at 0 0.0931>,<at 0.6099 0.0194>]}
       RedCurve {keys [<at 0 0.0437>,<at 0.6099 0.006>]}
      }})
    R1 ForestGreen:LinearColor=&ColorFromCurve(ColorCurve=DataInterface() {propertyValues {
       AlphaCurve {keys [<at 0 4 RCIM_Cubic>,<at 1 0 RCIM_Cubic>]}
       BlueCurve {keys [<at 0.0075 0.0716>,<at 0.5485 0.0091>]}
       GreenCurve {keys [<at 0.0075 0.3438>,<at 0.5485 0.067>]}
       RedCurve {keys [<at 0.0075 0.0573>,<at 0.5485 0.0061>]}
      }})
    R1 Grey:LinearColor=&ColorFromCurve(ColorCurve=DataInterface() {propertyValues {
       AlphaCurve {keys [<at 0 1 RCIM_Cubic>,<at 0.8868 0.9>,<at 0.9795 0 RCIM_Cubic>]}
       BlueCurve {keys [<at 0 0.2527>,<at 0.5743 0.067>]}
       GreenCurve {keys [<at 0 0.2527>,<at 0.5743 0.067>]}
       RedCurve {keys [<at 0 0.2527>,<at 0.5743 0.067>]}
      }})
    R1 Lime:LinearColor=&ColorFromCurve(ColorCurve=DataInterface() {propertyValues {
       AlphaCurve {keys [<at 0 1 RCIM_Cubic>,<at 0.8997 0.9>,<at 1 0 RCIM_Cubic>]}
       BlueCurve {keys [<at 0.014 0.0508>,<at 0.611 0.0061>]}
       GreenCurve {keys [<at 0.014 0.6435>,<at 0.611 0.067>]}
       RedCurve {keys [<at 0.014 0.344>,<at 0.611 0.0362>]}
      }})
    R1 Orange:LinearColor=&ColorFromCurve(ColorCurve=DataInterface() {propertyValues {
       AlphaCurve {keys [<at 0 1 RCIM_Cubic>,<at 0.9213 0.986>,<at 1 0 RCIM_Cubic>]}
       BlueCurve {keys [<at 0 0.0705 RCIM_Cubic>,<at 0.6217 0.0077>]}
       GreenCurve {keys [<at 0 0.357 RCIM_Cubic>,<at 0.6217 0.134>]}
       RedCurve {keys [<at 0 0.892 RCIM_Cubic>,<at 0.6217 0.3698>]}
      }})
    R1 Pink:LinearColor=&ColorFromCurve(ColorCurve=DataInterface() {propertyValues {
       AlphaCurve {keys [<at 0 1 RCIM_Cubic>,<at 0.9277 0.9949>,<at 1 0 RCIM_Cubic>]}
       BlueCurve {keys [<at 0.0334 0.4174>,<at 0.6282 0.1219>]}
       GreenCurve {keys [<at 0.0334 0.0991>,<at 0.6282 0.0166>]}
       RedCurve {keys [<at 0.0334 0.6>,<at 0.6282 0.1823>]}
      }})
    R1 Purple:LinearColor=&ColorFromCurve(ColorCurve=DataInterface() {propertyValues {
       AlphaCurve {keys [<at 0 1 RCIM_Cubic>,<at 0.8976 0.9795>,<at 1 0 RCIM_Cubic>]}
       BlueCurve {keys [<at 0.042 0.552>,<at 0.6239 0.1042>]}
       GreenCurve {keys [<at 0.042 0.0436>,<at 0.6239 0.0095>]}
       RedCurve {keys [<at 0.042 0.2523>,<at 0.6239 0.0483>]}
      }})
    R1 Saffron:LinearColor=&ColorFromCurve(ColorCurve=DataInterface() {propertyValues {
       AlphaCurve {keys [<at 0 1 RCIM_Cubic>,<at 0.9083 0.9979>,<at 1 0 RCIM_Cubic>]}
       BlueCurve {keys [<at 0.0075 0.0752>,<at 0.6174 0.0171>]}
       GreenCurve {keys [<at 0.0075 0.747>,<at 0.6174 0.2554>]}
       RedCurve {keys [<at 0.0075 0.952>,<at 0.6174 0.3281>]}
      }})
    R1 SkyBlue:LinearColor=&ColorFromCurve(ColorCurve=DataInterface() {propertyValues {
       AlphaCurve {keys [<at 0 1 RCIM_Cubic>,<at 0.9428 0.9815>,<at 1 0 RCIM_Cubic>]}
       BlueCurve {keys [<at 0 0.552>,<at 0.6088 0.067>]}
       GreenCurve {keys [<at 0 0.3834>,<at 0.6088 0.0468>]}
       RedCurve {keys [<at 0 0.0436>,<at 0.6088 0.0061>]}
      }})
    R1 TitaniumWhite:LinearColor=&ColorFromCurve(ColorCurve=DataInterface() {propertyValues {
       AlphaCurve {keys [<at 0 1 RCIM_Cubic>,<at 1 0 RCIM_Cubic>]}
       BlueCurve {keys [<at 0 0.6406>,<at 0.5808 0.2292>]}
       GreenCurve {keys [<at 0 0.6406>,<at 0.5808 0.2292>]}
       RedCurve {keys [<at 0 0.6406>,<at 0.5808 0.2292>]}
      }})
    R2 Black:LinearColor=&ColorFromCurve(ColorCurve=DataInterface() {propertyValues {
       AlphaCurve {keys [
         <at 0 0>
         {arriveTangent=3.3444, interpMode=RCIM_Cubic, leaveTangent=3.3444, time=0.0636, value=1}
         <at 0.3588 1.2>
         <at 0.6497 0 RCIM_Cubic>
        ]}
       BlueCurve {keys [<at 0.0463 0.1354>,<at 0.4752 0.7188>]}
       GreenCurve {keys [<at 0.0463 0.1354>,<at 0.4752 0.7188>]}
       RedCurve {keys [<at 0.0463 0.1354>,<at 0.4752 0.7188>]}
      }})
    R2 BurntSienna:LinearColor=&ColorFromCurve(ColorCurve=DataInterface() {propertyValues {
       AlphaCurve {keys [<at 0 0>, {arriveTangent=5.3491, interpMode=RCIM_Cubic, leaveTangent=5.3491, time=0.0442, value=3}, <at 0.3739 2>, <at 0.5959 0 RCIM_Cubic>]}
       BlueCurve {keys [<at 0.0248 0.115 RCIM_Cubic>,<at 0.3868 0.2014>]}
       GreenCurve {keys [<at 0.0248 0.1634 RCIM_Cubic>,<at 0.3868 0.2311>]}
       RedCurve {keys [<at 0.0248 0.3391 RCIM_Cubic>,<at 0.3868 0.3391>]}
      }})
    R2 Cobalt:LinearColor=&ColorFromCurve(ColorCurve=DataInterface() {propertyValues {
       AlphaCurve {keys [<at 0 0 RCIM_Cubic>,<at 0.0356 3>,<at 0.3761 2.5>,<at 0.6066 0 RCIM_Cubic>]}
       BlueCurve {keys [<at 0.0894 0.88>,<at 0.4256 3>]}
       GreenCurve {keys [<at 0.0894 0.0245>,<at 0.4256 0.1875>]}
       RedCurve {keys [<at 0.0894 0.0176>,<at 0.4256 0.1875>]}
      }})
    R2 Crimson:LinearColor=&ColorFromCurve(ColorCurve=DataInterface() {propertyValues {
       AlphaCurve {keys [<at 0 0>,<at 0.0442 1 RCIM_Cubic>,<at 0.2791 2.5>,<at 0.5916 0 RCIM_Cubic>]}
       BlueCurve {keys [<at 0 0.0405 RCIM_Cubic>,<at 0.3427 0.2462>,<at 0.4881 0.3841>]}
       GreenCurve {keys [<at 0 0.0176 RCIM_Cubic>,<at 0.3427 0.2307>,<at 0.4881 0.34>]}
       RedCurve {keys [<at 0 0.88 RCIM_Cubic>,<at 0.3427 0.855>,<at 0.4881 2>]}
      }})
    R2 DefaultColor:LinearColor=&ColorFromCurve(ColorCurve=DataInterface() {propertyValues {
       AlphaCurve {keys [<at 0.0498 0 RCIM_Cubic>,<at 0.1847 3>,<at 0.3787 3>,<at 0.5873 0 RCIM_Cubic>]}
       BlueCurve {keys [
         <at 0.0572 0.3117>
         {arriveTangent=1.3309, interpMode=RCIM_Cubic, leaveTangent=1.3309, time=0.195, value=0.6476}
         <at 0.3881 0.752>
         <at 0.5503 2 RCIM_Cubic>
        ]}
       GreenCurve {keys [
         <at 0.0572 1.0136>
         {arriveTangent=-0.9401, interpMode=RCIM_Cubic, leaveTangent=-0.9401, time=0.195, value=1.3642}
         <at 0.3881 0.7026>
         <at 0.5503 1.1655 RCIM_Cubic>
        ]}
       RedCurve {keys [
         <at 0.0572 1.76>
         {arriveTangent=-3.7354, interpMode=RCIM_Cubic, leaveTangent=-3.7354, time=0.195, value=1.594}
         <at 0.3881 0.5241>
         <at 0.5503 0.6771 RCIM_Cubic>
        ]}
      }})
    R2 ForestGreen:LinearColor=&ColorFromCurve(ColorCurve=DataInterface() {propertyValues {
       AlphaCurve {keys [<at 0 0>, {arriveTangent=6.0856, interpMode=RCIM_Cubic, leaveTangent=6.0856, time=0.0485, value=2}, <at 0.3286 2>, <at 0.5916 0 RCIM_Cubic>]}
       BlueCurve {keys [<at 0.0463 0.0608 RCIM_Cubic>,<at 0.4493 0.4231>]}
       GreenCurve {keys [<at 0.0463 0.88 RCIM_Cubic>,<at 0.4493 2>]}
       RedCurve {keys [<at 0.0463 0.0176 RCIM_Cubic>,<at 0.4493 0.34>]}
      }})
    R2 Grey:LinearColor=&ColorFromCurve(ColorCurve=DataInterface() {propertyValues {
       AlphaCurve {keys [
         <at 0.0119 0>
         {arriveTangent=2.9746, interpMode=RCIM_Cubic, leaveTangent=2.9746, time=0.0528, value=1}
         <at 0.348 1>
         <at 0.5808 0 RCIM_Cubic>
        ]}
       BlueCurve {keys [<at 0 0.4167 RCIM_Cubic>,<at 0.4429 2>]}
       GreenCurve {keys [<at 0 0.4167 RCIM_Cubic>,<at 0.4429 2>]}
       RedCurve {keys [<at 0 0.4167 RCIM_Cubic>,<at 0.4429 2>]}
      }})
    R2 Lime:LinearColor=&ColorFromCurve(ColorCurve=DataInterface() {propertyValues {
       AlphaCurve {keys [
         <at 0.0183 0>
         {arriveTangent=6.9779, interpMode=RCIM_Cubic, leaveTangent=6.9779, time=0.0593, value=1}
         <at 0.3049 2>
         <at 0.57 0 RCIM_Cubic>
        ]}
       BlueCurve {keys [<at 0.083 0.015 RCIM_Cubic>,<at 0.4924 0.34>]}
       GreenCurve {keys [<at 0.083 0.88 RCIM_Cubic>,<at 0.4924 2>]}
       RedCurve {keys [<at 0.083 0.4487 RCIM_Cubic>,<at 0.4924 1.1723>]}
      }})
    R2 Orange:LinearColor=&ColorFromCurve(ColorCurve=DataInterface() {propertyValues {
       AlphaCurve {keys [
         <at 0.0205 0>
         {arriveTangent=5.5907, interpMode=RCIM_Cubic, leaveTangent=5.5907, time=0.1261, value=2.5}
         <at 0.3782 2>
         <at 0.6476 0 RCIM_Cubic>
        ]}
       BlueCurve {keys [<at 0.0657 0.015 RCIM_Cubic>,<at 0.4881 0.165>]}
       GreenCurve {keys [<at 0.0657 0.3167 RCIM_Cubic>,<at 0.4881 0.4563>]}
       RedCurve {keys [<at 0.0657 0.88 RCIM_Cubic>,<at 0.4881 1>]}
      }})
    R2 Pink:LinearColor=&ColorFromCurve(ColorCurve=DataInterface() {propertyValues {
       AlphaCurve {keys [
         <at 0.0183 0>
         {arriveTangent=4.1431, interpMode=RCIM_Cubic, leaveTangent=4.1431, time=0.083, value=1.5}
         <at 0.3804 1.5>
         <at 0.6002 0 RCIM_Cubic>
        ]}
       BlueCurve {keys [<at 0.0614 0.5654 RCIM_Cubic>,<at 0.5118 1.3913>]}
       GreenCurve {keys [<at 0.0614 0.015 RCIM_Cubic>,<at 0.5118 0.33>]}
       RedCurve {keys [<at 0.0614 0.88 RCIM_Cubic>,<at 0.5118 2>]}
      }})
    R2 Purple:LinearColor=&ColorFromCurve(ColorCurve=DataInterface() {propertyValues {
       AlphaCurve {keys [<at 0 0>, {arriveTangent=5.2581, interpMode=RCIM_Cubic, leaveTangent=5.2581, time=0.0679, value=1}, <at 0.3804 2>, <at 0.6325 0 RCIM_Cubic>]}
       BlueCurve {keys [<at 0.0528 0.88 RCIM_Cubic>,<at 0.6002 2>]}
       GreenCurve {keys [<at 0.0528 0.015 RCIM_Cubic>,<at 0.6002 0.1562>]}
       RedCurve {keys [<at 0.0528 0.37 RCIM_Cubic>,<at 0.6002 0.913>]}
      }})
    R2 Saffron:LinearColor=&ColorFromCurve(ColorCurve=DataInterface() {propertyValues {
       AlphaCurve {keys [<at 0.0291 0>,<at 0.0959 1.5>,<at 0.4601 1.5>,<at 0.57 0 RCIM_Cubic>]}
       BlueCurve {keys [<at 0.07 0.015>,<at 0.5054 0.33>]}
       GreenCurve {keys [<at 0.07 0.6777>,<at 0.5054 1.6095>]}
       RedCurve {keys [<at 0.07 0.88>,<at 0.5054 2>]}
      }})
    R2 SkyBlue:LinearColor=&ColorFromCurve(ColorCurve=DataInterface() {propertyValues {
       AlphaCurve {keys [<at 0 0 RCIM_Cubic>,<at 0.0959 1.9879>,<at 0.3157 1.995>,<at 0.4321 2>,<at 0.6066 0 RCIM_Cubic>]}
       BlueCurve {keys [<at 0.0916 0.88 RCIM_Cubic>,<at 0.5334 2>]}
       GreenCurve {keys [<at 0.0916 0.5931 RCIM_Cubic>,<at 0.5334 1.4461>]}
       RedCurve {keys [<at 0.0916 0.015 RCIM_Cubic>,<at 0.5334 0.33>]}
      }})
    R2 TitaniumWhite:LinearColor=&ColorFromCurve(ColorCurve=DataInterface() {propertyValues {
       AlphaCurve {keys [<at 0 0 RCIM_Cubic>,<at 0.0679 1.4>,<at 0.3653 1.5>,<at 0.5937 0 RCIM_Cubic>]}
       BlueCurve {keys [<at 0.0463 0.5573 RCIM_Cubic>,<at 0.4816 2>]}
       GreenCurve {keys [<at 0.0463 0.5573 RCIM_Cubic>,<at 0.4816 2>]}
       RedCurve {keys [<at 0.0463 0.5573 RCIM_Cubic>,<at 0.4816 2>]}
      }})
    R3 Black:LinearColor=<rgba 0.0052 0.0052 0.0052 1>
    R3 BurntSienna:LinearColor=<rgba 0.0312 0.0096 0.0039 1>
    R3 Cobalt:LinearColor=<rgba 0 0.004 0.2135 1>
    R3 Crimson:LinearColor=<rgba 0.1458 0.0114 0.015 1>
    R3 DefaultAlpha=4
    R3 DefaultColor:LinearColor=<rgba 0.0011 0.0034 0.0208 1>
    R3 ForestGreen:LinearColor=<rgba 0 0.0781 0.0039 1>
    R3 Grey:LinearColor=<rgba 0.1042 0.1042 0.1042 1>
    R3 Lime:LinearColor=<rgba 0.0677 0.1354 0 1>
    R3 Orange:LinearColor=<rgba 0.3177 0.1108 0 1>
    R3 Pink:LinearColor=<rgba 0.2344 0.0281 0.1594 1>
    R3 Purple:LinearColor=<rgba 0.062 0 0.151 1>
    R3 Saffron:LinearColor=<rgba 0.2188 0.1677 0 1>
    R3 SkyBlue:LinearColor=<rgba 0 0.1213 0.1823 1>
    R3 TitaniumWhite:LinearColor=<rgba 0.1406 0.1406 0.1406 1>
    Use R1 Painted Overrides=true
    Use R2 Painted Overrides=true
    Use R3 Painted Overrides=true
   ) {module=VC_PaintedVariantsOverride001}
   &ParticleState()
   &DynamicMaterialParameters(
    Index 0 Param 4=2.3102
    Module.Index0Float0Write=false
    Module.Index0Float1Write=false
    Module.Index0Float2Write=false
    Module.Index0Float3Write=true
   )
   &InheritSourceMovement(ApplyForce=false, ApplyPosition=true, ApplyPositionFalloff=false) {module=InheritSourceMovement002}
   &SolveForcesAndVelocity(Clamp Velocity=false, Limit Acceleration=false)
   /SetVariables(Particles.DotProduct:NiagaraFloat=&DotProduct(
     Vector A=/Niagara/DynamicInputs/Math/TransformVectorByMatrix(
      Vector:Vector3f=<1 0 0>
      Transformation Matrix:NiagaraMatrix=$Engine.Owner.SystemLocalToWorld
      Transform Vector as Position:NiagaraBool=false
     )
     Vector B=&NormalizeVector(
      Vector To Normalize=&SubtractVectorFromPosition(
       Position=$Engine.Owner.Position
       Vector=&ReturnCameraPosition(Coordinate Space:ENiagaraCoordinateSpace=::Simulation, Camera Query:NiagaraDataInterfaceCamera=DataInterface())
      )
      Safe Normalize=true
      Safe Fallback Threshold=0
      Safe Fallback Vector=<0 0 0.001>
     )
    )) {isSetParameters=true}
   /SetVariables(Particles.CustomPosition_R1:NiagaraPosition=&AddVectorToPosition(
     Position=$Particles.Position
     Vector=&MakeVector(
      X=0
      Y=0
      Z=&MaskFloat(
       Mask Source=0
       Mask Parameter=?
       Mask=0
       Masked Float=&Cosine(
        Normalized Angle=&Add_Float(A=0, B=&Subtract_Float(A=$Emitter.NormalizedLoopAge, B=$Particles.NormalizedAge))
        Period=1
        Scale=&FloatFromCurve(
         FloatCurve=DataInterface() {propertyValues {Curve {keys [
             {arriveTangent=4.1101, interpMode=RCIM_Cubic, leaveTangent=4.1101, tangentMode=RCTM_User, time=0.0999, value=0.71}
             {arriveTangent=3.0214, interpMode=RCIM_Cubic, leaveTangent=3.0214, time=0.4103, value=3.8}
             {
              arriveTangent=-0.5881
              arriveTangentWeight=0.5973
              interpMode=RCIM_Cubic
              leaveTangent=-0.5881
              leaveTangentWeight=0.5973
              tangentMode=RCTM_User
              time=0.7585
              value=2.7
             }
            ]}}}
         Scale Curve=&Lerp_Float(
          A=10
          B=6.25
          Alpha=&NormalizeFloat(Value=&Power(a=&Add_Float(A=$Particles.DotProduct, B=0.1), b=5), Min=0, Max=1)
         )
        )
        Bias=0
       )
       Passthrough Float=&Sine(
        Normalized Angle=&Add_Float(A=0, B=&Subtract_Float(A=$Emitter.NormalizedLoopAge, B=$Particles.NormalizedAge))
        Period=1
        Scale=&FloatFromCurve(
         FloatCurve=DataInterface() {propertyValues {Curve {keys [
             {arriveTangent=4.1101, interpMode=RCIM_Cubic, leaveTangent=4.1101, tangentMode=RCTM_User, time=0.0999, value=0.71}
             {arriveTangent=3.0214, interpMode=RCIM_Cubic, leaveTangent=3.0214, time=0.4103, value=3.8}
             {
              arriveTangent=-0.5881
              arriveTangentWeight=0.5973
              interpMode=RCIM_Cubic
              leaveTangent=-0.5881
              leaveTangentWeight=0.5973
              tangentMode=RCTM_User
              time=0.7585
              value=2.7
             }
            ]}}}
         Scale Curve=&Lerp_Float(
          A=10
          B=6.25
          Alpha=&NormalizeFloat(Value=&Power(a=&Add_Float(A=$Particles.DotProduct, B=0.1), b=5), Min=0, Max=1)
         )
        )
        Bias=0
       )
      )
     )
    )) {isSetParameters=true}
   &VC_UpdateRibbon_v1(Enable Skeletal Position=true)
   /SetVariables(
    Particles.R1_Width:NiagaraFloat=&FloatFromCurve(
     FloatCurve=DataInterface() {propertyValues {Curve {keys [
         {interpMode=RCIM_Cubic, tangentMode=RCTM_Break, time=0.0099, value=0.7479}
         {arriveTangent=0.7072, interpMode=RCIM_Cubic, leaveTangent=0.7072, time=0.2552, value=1.2361}
         {arriveTangent=-1.1089, interpMode=RCIM_Cubic, leaveTangent=-1.1089, time=0.5313, value=1.1166}
         {interpMode=RCIM_Cubic, tangentMode=RCTM_Break, time=0.9971, value=0.4134}
        ]}}}
     Scale Curve=112
    )
    Particles.R2_Width:NiagaraFloat=&FloatFromCurve(
     FloatCurve=DataInterface() {propertyValues {Curve {keys [
         {interpMode=RCIM_Cubic, tangentMode=RCTM_Break, time=0.0118, value=0.7019}
         {arriveTangent=0.3522, interpMode=RCIM_Cubic, leaveTangent=0.3522, tangentMode=RCTM_User, time=0.2176, value=0.8787}
         {arriveTangent=-0.5969, interpMode=RCIM_Cubic, leaveTangent=-0.5969, time=0.6765, value=0.7063}
         {interpMode=RCIM_Cubic, tangentMode=RCTM_Break, time=0.9971, value=0.4134}
        ]}}}
     Scale Curve=50
    )
    Particles.R3_Width:NiagaraFloat=&FloatFromCurve(
     FloatCurve=DataInterface() {propertyValues {Curve {keys [
         {interpMode=RCIM_Cubic, tangentMode=RCTM_Break, time=0.0327, value=0.8536}
         {arriveTangent=0.1641, interpMode=RCIM_Cubic, leaveTangent=0.1641, time=0.2819, value=1.2814}
         {arriveTangent=-1.298, interpMode=RCIM_Cubic, leaveTangent=-1.298, time=0.64, value=0.9532}
         {interpMode=RCIM_Cubic, tangentMode=RCTM_Break, time=0.9971, value=0.353}
        ]}}}
     Scale Curve=120
    )
   ) {isSetParameters=true}
  ]
  rendererClasses [&NiagaraRibbonRendererProperties]
  renderers [
   {rendererClass=&NiagaraRibbonRendererProperties}
   {rendererClass=&NiagaraRibbonRendererProperties, rendererIndex=1}
   {rendererClass=&NiagaraRibbonRendererProperties, rendererIndex=2}
  ]
 }
 NE_VC_Boost_SpawnRate {
  emitterUpdateScript [
   &EmitterState(Inactive Response=::NewEnumerator1, Life Cycle Mode=::NewEnumerator1, Loop Duration=1, UseLoopDelay=false)
   &VC_FilteredSocketCount()
   &SpawnRate(SpawnRate=20, Use Spawn Probability=false)
  ]
  particleSpawnScript [
   &VC_PaintedVariantsOverride(
    Black=<rgba 0.0625 0.0625 0.0625 1>
    BurntSienna=<rgba 0.1771 0.068 0.0387 1>
    Cobalt=<rgba 0.0479 0.0618 0.799 1>
    Crimson=<rgba 0.799 0.0479 0.0683 1>
    DefaultColor=<rgba 0.7991 0.6172 0.0482 1>
    ForestGreen=<rgba 0.0495 0.3391 0.0641 1>
    Grey=<rgba 0.2087 0.2087 0.2087 1>
    Lime=<rgba 0.4235 0.799 0.0479 1>
    Orange=<rgba 0.799 0.3099 0.0479 1>
    Pink=<rgba 0.799 0.0479 0.5259 1>
    Purple=<rgba 0.3443 0.0278 0.799 1>
    Saffron=<rgba 0.799 0.6304 0.0764 1>
    SkyBlue=<rgba 0.0479 0.5477 0.799 1>
    TitaniumWhite=<rgba 0.3698 0.3698 0.3698 1>
   )
   &InitializeParticle(
    Color=$Particles.Color
    Lifetime=&MergeHelper_FloatSwitch(0=0.5, 1=0.5, 2=?, 3=?, Index=$User.GameModeID, Num=::NewEnumerator0)
    Sprite Rotation Mode=::NewEnumerator2
    Sprite Size Mode=::NewEnumerator3
    Uniform Sprite Size=80
    UsePositionOffset=false
    Write Material Random=true
   )
   &PartitionParticles(Number Of Partitions=$Emitter.PortCount)
   &SkeletalMeshLocation(
    Bone / Socket Index=$Particles.PartitionParticles.Partition
    Bone / Socket Sampling Mode=::NewEnumerator5
    MaskSpawnGroup=false
    Mesh Sampling Type=::NewEnumerator6
    Sampled Position Offset=&AddVectorToPosition(Position=$User.DTCarOffset002, Vector=<-46 0 0>)
    Skeletal Mesh=$System.SkeletalMesh
   )
   &AddVelocity(Velocity=<-1500 0 0>) {enabled=false}
  ]
  particleUpdateScript [
   &ParticleState()
   &SolveForcesAndVelocity(Clamp Velocity=false, Limit Acceleration=false)
  ]
  rendererClasses [&NiagaraSpriteRendererProperties]
  renderers [{rendererClass=&NiagaraSpriteRendererProperties}]
 }
 NE_VC_Boost_SpawnRate001 {
  emitterUpdateScript [
   &EmitterState()
   &VC_FilteredSocketCount()
   &SpawnRate(SpawnGroup=1, SpawnRate=7, Use Spawn Probability=true) {module=SpawnRate001}
   &SpawnRate(SpawnRate=60, Use Spawn Probability=false) {enabled=false}
  ]
  particleSpawnScript [
   &VC_PaintedVariantsOverride(
    Black=<rgba 0.0938 0.0938 0.0938 1>
    BurntSienna=<rgba 0.2086 0.1301 0.1095 1>
    Cobalt=<rgba 0.0249 0.0398 0.83 1>
    Crimson=<rgba 0.83 0.0249 0.0467 1>
    DefaultColor=<rgba 0.8388 0.4735 0.0185 1>
    ForestGreen=<rgba 0.0185 0.839 0.0599 1>
    Grey=<rgba 0.3565 0.3565 0.3565 1>
    Lime=<rgba 0.4287 0.839 0.0185 1>
    Orange=<rgba 0.839 0.3307 0.0584 1>
    Pink=<rgba 0.839 0.1442 0.5863 1>
    Purple=<rgba 0.3552 0.0185 0.839 1>
    Saffron=<rgba 0.839 0.6857 0.1835 1>
    SkyBlue=<rgba 0.0918 0.589 0.839 1>
    TitaniumWhite=<rgba 1 1 1 1>
   )
   &InitializeParticle(
    Color=$Particles.Color
    Lifetime=&MergeHelper_FloatSwitch(0=0.6, 1=0.6, 2=0.6, 3=?, Index=$User.GameModeID, Num=::NewEnumerator1)
    Position Mode=::NewEnumerator0
    Sprite Rotation Angle=&RandomRangeFloat(Minimum=-3.255, Maximum=7.291)
    Sprite Rotation Mode=::NewEnumerator1
    Sprite Size Mode=::NewEnumerator1
    Uniform Sprite Size Max=60
    Uniform Sprite Size Min=25
    Write Material Random=true
   )
   &PartitionParticles(Particles Per Partition=1, Partition Mode:ENiagara_PartitionMode=::NewEnumerator1, Write Partition to Ribbon ID=::NewEnumerator0)
   &SkeletalMeshLocation(
    Bone / Socket Index=$Particles.PartitionParticles.Partition
    Bone / Socket Sampling Mode=::NewEnumerator5
    MaskSpawnGroup=false
    Mesh Sampling Type=::NewEnumerator6
    Sampled Position Offset=&AddVectorToPosition(Position=<0 0 0>, Vector=<-70 0 0>)
    Skeletal Mesh=$System.SkeletalMesh
   )
   &AddVelocity(Velocity=&MergeHelper_VectorSwitch(0=<-487 0 0>, 1=<-487 0 0>, 2=<-487 0 0>, 3=?, Index=$User.GameModeID, Num=::NewEnumerator1))
   &ShapeLocation(
    Cylinder Height=66.1754
    Cylinder Radius=&FloatFromCurve(
     FloatCurve=DataInterface() {propertyValues {Curve {defaultValue=1, keys [{tangentMode=RCTM_User, time=-0.0027, value=0.7292}, {tangentMode=RCTM_User, time=0.25, value=2}]}}}
     Scale Curve=62
    )
    MaskSpawnGroup=false
    Override Seed=false
    Shape Primitive=::NewEnumerator1
    Surface Only=false
    UseOwnerScale=true
   )
  ]
  particleUpdateScript [
   &ParticleState()
   /SetVariables(
    Particles.Stars_RendererTag:NiagaraInt32=&SpawnGroup()
    Particles.Stars_Size:Vector2f=&RandomRangeVector2D(Minimum=<10 10>, Maximum=<50 50>)
   ) {isSetParameters=true}
   &SpriteRotationRate(Rotation Rate=172.7809)
   &SolveForcesAndVelocity(Clamp Velocity=false, Limit Acceleration=false)
   &ScaleColor(
    Scale Alpha=&FloatFromCurve(FloatCurve=DataInterface() {propertyValues {Curve {keys [
         {interpMode=RCIM_Cubic, tangentMode=RCTM_Break, time=0, value=1}
         {interpMode=RCIM_Cubic, tangentMode=RCTM_Break, time=0.3357, value=1.3539}
         {interpMode=RCIM_Cubic, tangentMode=RCTM_Break, time=1, value=0}
        ]}}})
    ScaleA=true
    ScaleRGB=true
   )
  ]
  rendererClasses [&NiagaraSpriteRendererProperties]
  renderers [
   {rendererClass=&NiagaraSpriteRendererProperties}
   {rendererClass=&NiagaraSpriteRendererProperties, rendererIndex=1}
  ]
 }
}
summary {
 systemName=NS_StarryNight
 emitters [
  {emitterName=NE_VC_Boost_Ribbons, rendererClasses [&NiagaraRibbonRendererProperties]}
  {emitterName=NE_VC_Boost_SpawnRate, rendererClasses [&NiagaraSpriteRendererProperties]}
  {emitterName=NE_VC_Boost_SpawnRate001, rendererClasses [&NiagaraSpriteRendererProperties]}
  {
   emitterName=NE_VC_Boost_MeshCone001
   rendererClasses [&NiagaraMeshRendererProperties, &NiagaraSpriteRendererProperties]
  }
 ]
 userVariables [
  User.bIsFrontendPreview {type=NiagaraBool, defaultValue=true}
  User.bIsPainted {type=NiagaraBool}
  User.ColorVariantID {type=NiagaraInt32}
  User.DTCarOffset002 {type=NiagaraPosition, description="", defaultValue {}}
  User.GameModeID {type=NiagaraInt32}
  User.PaintedColor {type=LinearColor, description="", defaultValue {struct=LinearColor, value=<rgba 1 1 1 1>}}
  User.PaintedScalar {type=NiagaraFloat, defaultValue=1, description=""}
  User.PortCount {type=NiagaraInt32, defaultValue=2}
 ]
}
